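# Animation Blueprint for the Reverb backpack cosmetic that makes its speaker bones pulse procedurally.
# Each update, speakerReverbFNC layers game-time Perlin noise and sine into offset and scale values for speaker and bar variables, with differing frequencies and ranges.
# Additive bone-space modifiers apply these to both speakers over the input pose.
# The bar-bone modifier chain and Copy Pose From Mesh node are not connected to the output pose.
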AnimBlueprint /BRCosmetics/Accessories/FORT_Backpacks/F_MED_Reverb/Meshes/F_MED_Reverb_Pack_AnimBP
target_skeleton=&/F_MED_Reverb_Pack_Skeleton
graphs [
 AnimGraph {nodes [
   Copy Pose From Mesh() {bUseAttachedParent=true}
   n1 = "Transform (Modify) Bone - Bone: dyn_dn"(Alpha=1, Translation="0,0,0", Rotation="0,0,0", Scale=Get bar_a() -> (bar_a:Vector)) {
    rotationMode=BMM_Additive
    rotationSpace=BCS_BoneSpace
    scaleMode=BMM_Additive
    scaleSpace=BCS_BoneSpace
    translationMode=BMM_Additive
    translationSpace=BCS_BoneSpace
    alphaBoolBlend {blendOption=Linear}
    boneToModify {boneName=dyn_dn}
   }
   n2 = Get bar_b() -> (bar_b:Vector)
   n3 = "Transform (Modify) Bone - Bone: dyn_dn_a_l"(ComponentPose=@n1, Alpha=1, Translation="0,0,0", Rotation="0,0,0", Scale=@n2) {
    rotationMode=BMM_Additive
    rotationSpace=BCS_BoneSpace
    scaleMode=BMM_Additive
    scaleSpace=BCS_BoneSpace
    translationMode=BMM_Additive
    translationSpace=BCS_BoneSpace
    alphaBoolBlend {blendOption=Linear}
    boneToModify {boneName=dyn_dn_a_l}
   }
   n4 = "Transform (Modify) Bone - Bone: dyn_dn_b_r"(ComponentPose=@n3, Alpha=1, Translation="0,0,0", Rotation="0,0,0", Scale=@n2) {
    rotationMode=BMM_Additive
    rotationSpace=BCS_BoneSpace
    scaleMode=BMM_Additive
    scaleSpace=BCS_BoneSpace
    translationMode=BMM_Additive
    translationSpace=BCS_BoneSpace
    alphaBoolBlend {blendOption=Linear}
    boneToModify {boneName=dyn_dn_b_r}
   }
   n5 = "Transform (Modify) Bone - Bone: dyn_up"(ComponentPose=@n4, Alpha=1, Translation="0,0,0", Rotation="0,0,0", Scale=Get bar_e() -> (bar_e:Vector)) {
    rotationMode=BMM_Additive
    rotationSpace=BCS_BoneSpace
    scaleMode=BMM_Additive
    scaleSpace=BCS_BoneSpace
    translationMode=BMM_Additive
    translationSpace=BCS_BoneSpace
    alphaBoolBlend {blendOption=Linear}
    boneToModify {boneName=dyn_up}
   }
   n6 = Get bar_c() -> (bar_c:Vector)
   n7 = "Transform (Modify) Bone - Bone: dyn_dn_b_l"(ComponentPose=@n5, Alpha=1, Translation="0,0,0", Rotation="0,0,0", Scale=@n6) {
    rotationMode=BMM_Additive
    rotationSpace=BCS_BoneSpace
    scaleMode=BMM_Additive
    scaleSpace=BCS_BoneSpace
    translationMode=BMM_Additive
    translationSpace=BCS_BoneSpace
    alphaBoolBlend {blendOption=Linear}
    boneToModify {boneName=dyn_dn_b_l}
   }
   "Transform (Modify) Bone - Bone: dyn_dn_a_r"(ComponentPose=@n7, Alpha=1, Translation="0,0,0", Rotation="0,0,0", Scale=@n6) {
    rotationMode=BMM_Additive
    rotationSpace=BCS_BoneSpace
    scaleMode=BMM_Additive
    scaleSpace=BCS_BoneSpace
    translationMode=BMM_Additive
    translationSpace=BCS_BoneSpace
    alphaBoolBlend {blendOption=Linear}
    boneToModify {boneName=dyn_dn_a_r}
   }
   AnimGraphNode_SubInput_0 = Input Pose()
   n8 = Get reverbTran() -> (reverbTran:Vector)
   n9 = Get reverbScale() -> (reverbScale:Vector)
   n10 = "Transform (Modify) Bone - Bone: dyn_speaker_l"(ComponentPose=Local To Component(LocalPose=@AnimGraphNode_SubInput_0), Alpha=1, Translation=@n8, Rotation="0,0,0", Scale=@n9) {
    rotationMode=BMM_Additive
    rotationSpace=BCS_BoneSpace
    scaleMode=BMM_Additive
    scaleSpace=BCS_BoneSpace
    translationMode=BMM_Additive
    translationSpace=BCS_BoneSpace
    alphaBoolBlend {blendOption=Linear}
    boneToModify {boneName=dyn_speaker_l}
   }
   n11 = "Transform (Modify) Bone - Bone: dyn_speaker_r"(ComponentPose=@n10, Alpha=1, Translation=@n8, Rotation="0,0,0", Scale=@n9) {
    rotationMode=BMM_Additive
    rotationSpace=BCS_BoneSpace
    scaleMode=BMM_Additive
    scaleSpace=BCS_BoneSpace
    translationMode=BMM_Additive
    translationSpace=BCS_BoneSpace
    alphaBoolBlend {blendOption=Linear}
    boneToModify {boneName=dyn_speaker_r}
   }
   Output Pose(Result=Component To Local(ComponentPose=@n11)) {layerGroup=None}
  ]}
 EventGraph {nodes [
   Event BlueprintUpdateAnimation()
   > n1 = Sequence!(execute) -> (then_0, then_1, then_2:Exec, then_3:Exec, then_4:Exec)
   > n2 = speakerReverbFNC!(execute, self, frequency:double=7, amp:double=0.2, scaleMin:double=0.95, scaleMax:double=1.05) -> (then, tran_output:double, scale_output:double)
   > Set reverbTran(reverbTran:Vector=MakeVector(X=@n2, Y=0)) -> (Output_Get:Vector)
   > Set reverbScale(reverbScale:Vector=MakeVector(X=@n2.scale_output, Y=@n2.scale_output, Z=@n2.scale_output)) -> (Output_Get:Vector)
   n3 = speakerReverbFNC(execute=@n1.then_1, frequency=4, amp=1, scaleMin=0.5, scaleMax=1.3)
   > Set bar_a(bar_a:Vector=MakeVector(X=@n3.scale_output, Y=1, Z=1)) -> (Output_Get:Vector)
   n4 = speakerReverbFNC(execute=@n1.then_2, frequency=5, amp=1, scaleMin=0.4, scaleMax=1.2)
   > Set bar_b(bar_b:Vector=MakeVector(X=@n4.scale_output, Y=1, Z=1)) -> (Output_Get:Vector)
   n5 = speakerReverbFNC(execute=@n1.then_3, frequency=6, amp=1, scaleMin=0.3, scaleMax=1.1)
   > Set bar_c(bar_c:Vector=MakeVector(X=@n5.scale_output, Y=1, Z=1)) -> (Output_Get:Vector)
   n6 = speakerReverbFNC(execute=@n1.then_4, frequency=6, amp=1, scaleMin=0.1, scaleMax=3)
   > Set bar_e(bar_e:Vector=MakeVector(X=@n6.scale_output, Y=1, Z=1)) -> (Output_Get:Vector)
   TryGetPawnOwner()
  ]}
 speakerReverbFNC {nodes [
   n1 = speakerReverbFNC!() -> (then, frequency:double=1, amp:double=1, scaleMin:double=0.9, scaleMax:double=1.1) {is=title K2Node_FunctionEntry}
   n2 = GetGameTimeInSeconds()
   n3 = "Sin (Radians)"(A=PerlinNoise1D!(Value:float=float * float(A=@n1, B=@n2)) -> (ReturnValue:float))
   > Return Node!(
    execute
    tran_output:double=PerlinNoise1D(Value=float * float(A=@n3, B=@n1.amp))
    scale_output:double=Make Perlin Noise and Remap!(Value:float=@n3, RangeOutMin:float=@n1.scaleMin, RangeOutMax:float=@n1.scaleMax) -> (ReturnValue:float)
   )
  ]}
]
variables [
 bar_a {type=Vector}
 bar_b {type=Vector}
 bar_c {type=Vector}
 bar_e {type=Vector}
 reverbScale {type=Vector}
 reverbTran {type=Vector}
]
